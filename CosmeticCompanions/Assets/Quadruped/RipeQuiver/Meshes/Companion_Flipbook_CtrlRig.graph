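# Control Rig for a quadruped cosmetic companion that converts facial pose curves into left and right eye flipbook index curves.
# Each update resets both indices, then thresholds blink, eye-wide and brow curves to pick a frame per eye.
# Chained branches let later brow or wide matches override a half-blink, while a full blink skips them and checks only brow-down.
# The right eye's brow flags read the left brow curves, leaving R_brow curves unused.
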ControlRigBlueprint /CosmeticCompanions/Assets/Quadruped/RipeQuiver/Meshes/Companion_Flipbook_CtrlRig
graphs [
 New Function_ContainedGraph {nodes [
   n1 = RigVMFunctionEntryNode()
   GreaterEqual = /Script/RigVM.RigVMFunction_MathDoubleGreaterEqual(B=@n1.Threshold, A=/Script/ControlRig.RigUnit_GetCurveValue(Curve=@n1.Inputcurve.Name) {is=struct RigVMUnitNode}.Value) {is=struct RigVMUnitNode, template="GreaterEqual::Execute(in A,in B,out Result)"}
   > RigVMFunctionReturnNode(Activated=@GreaterEqual.Result)
  ]}
 RigVMFunctionLibrary {nodes [IsCurveActivatedFNC = RigVMCollapseNode(ExecuteContext=(), Inputcurve=(Type=None,Name="None"))]}
 RigVMFunction_MathBoolOr_1_ContainedGraph {nodes [
   n1 = RigVMFunctionEntryNode()
   RigVMFunction_MathBoolOr = &RigVMFunction_MathBoolOr(A=@n1.A, B=@n1.B)
   RigVMFunctionReturnNode(Result=&RigVMFunction_MathBoolOr(B=@n1.C, A=@RigVMFunction_MathBoolOr.Result).Result)
  ]}
 RigVMFunction_MathBoolOr_1_ContainedGraph {nodes [
   n1 = RigVMFunctionEntryNode()
   RigVMFunction_MathBoolOr = &RigVMFunction_MathBoolOr(A=@n1.A, B=@n1.B)
   RigVMFunctionReturnNode(Result=&RigVMFunction_MathBoolOr(B=@n1.C, A=@RigVMFunction_MathBoolOr.Result).Result)
  ]}
 RigVMFunction_Sequence_2_ContainedGraph {nodes [
   n1 = RigVMFunctionEntryNode()
   n2 = &RigVMFunction_Sequence(ExecuteContext=@n1.ExecuteContext)
   n3 = &RigVMFunction_Sequence(ExecuteContext=@n2.B)
   n4 = &RigVMFunction_Sequence(ExecuteContext=@n3.B)
   RigVMFunctionReturnNode(A=@n2.A, B=@n3.A, C=@n4.A, D=@n4.B)
  ]}
 RigVMFunction_Sequence_2_ContainedGraph {nodes [
   n1 = RigVMFunctionEntryNode()
   n2 = &RigVMFunction_Sequence(ExecuteContext=@n1.ExecuteContext)
   n3 = &RigVMFunction_Sequence(ExecuteContext=@n2.B)
   n4 = &RigVMFunction_Sequence(ExecuteContext=@n3.B)
   RigVMFunctionReturnNode(A=@n2.A, B=@n3.A, C=@n4.A, D=@n4.B)
  ]}
 RigVMFunction_Sequence_4_ContainedGraph {nodes [
   n1 = RigVMFunctionEntryNode()
   RigVMFunction_Sequence = &RigVMFunction_Sequence(ExecuteContext=@n1.ExecuteContext)
   n2 = &RigVMFunction_Sequence(ExecuteContext=@RigVMFunction_Sequence.B)
   RigVMFunctionReturnNode(A=@RigVMFunction_Sequence.A, B=@n2.A, C=@n2.B)
  ]}
 RigVMModel {nodes [
   RigUnit_BeginExecution = &RigUnit_BeginExecution()
   /Script/ControlRig.RigUnit_SetCurveValue(Curve=flipbook_face_L_eye_index, Value=0, ExecutePin=@RigUnit_BeginExecution.ExecutePin) {is=struct RigVMUnitNode}
   > &RigUnit_SetCurveValue(Curve=flipbook_face_R_eye_index, Value=0)
   > RigVMFunction_Sequence = RigVMAggregateNode()
   n1 = &RigVMFunction_Sequence(ExecuteContext=@RigVMFunction_Sequence.A)
   IsCurveActivatedFNC_10 = RigVMFunctionReferenceNode(Inputcurve=(Type=Curve,Name="L_brow_In_Dn_pose"), ExecuteContext=@n1.A, Threshold=RigVMVariableNode() {variable=threshold}.Value)
   > RigVMVariableNode(Value=@IsCurveActivatedFNC_10.Activated) {variable=l_brow_in_dn}
   > IsCurveActivatedFNC_11 = RigVMFunctionReferenceNode(Inputcurve=(Type=Curve,Name="L_brow_In_Up_pose"), Threshold=RigVMVariableNode() {variable=threshold}.Value)
   > RigVMVariableNode(Value=@IsCurveActivatedFNC_11.Activated) {variable=l_brow_in_up}
   > IsCurveActivatedFNC_12 = RigVMFunctionReferenceNode(Inputcurve=(Type=Curve,Name="L_brow_Ot_Dn_pose"), Threshold=RigVMVariableNode() {variable=threshold}.Value)
   > RigVMVariableNode(Value=@IsCurveActivatedFNC_12.Activated) {variable=l_brow_out_dn}
   > IsCurveActivatedFNC_13 = RigVMFunctionReferenceNode(Inputcurve=(Type=Curve,Name="L_brow_Ot_Up_pose"), Threshold=RigVMVariableNode() {variable=threshold}.Value)
   > RigVMVariableNode(Value=@IsCurveActivatedFNC_13.Activated) {variable=l_brow_out_up}
   > RigVMVariableNode(Value=&RigUnit_GetCurveValue(Curve=L_blink_pose).Value) {variable=l_blink_val}
   IsCurveActivatedFNC_14 = RigVMFunctionReferenceNode(Inputcurve=(Type=Curve,Name="L_brow_In_Dn_pose"), ExecuteContext=@n1.B, Threshold=RigVMVariableNode() {variable=threshold}.Value)
   > RigVMVariableNode(Value=@IsCurveActivatedFNC_14.Activated) {variable=r_brow_in_dn}
   > IsCurveActivatedFNC_15 = RigVMFunctionReferenceNode(Inputcurve=(Type=Curve,Name="L_brow_In_Up_pose"), Threshold=RigVMVariableNode() {variable=threshold}.Value)
   > RigVMVariableNode(Value=@IsCurveActivatedFNC_15.Activated) {variable=r_brow_in_up}
   > IsCurveActivatedFNC_16 = RigVMFunctionReferenceNode(Inputcurve=(Type=Curve,Name="L_brow_Ot_Dn_pose"), Threshold=RigVMVariableNode() {variable=threshold}.Value)
   > RigVMVariableNode(Value=@IsCurveActivatedFNC_16.Activated) {variable=r_brow_out_dn}
   > IsCurveActivatedFNC_17 = RigVMFunctionReferenceNode(Inputcurve=(Type=Curve,Name="L_brow_Ot_Up_pose"), Threshold=RigVMVariableNode() {variable=threshold}.Value)
   > RigVMVariableNode(Value=@IsCurveActivatedFNC_17.Activated) {variable=r_brow_out_up}
   > RigVMVariableNode(Value=&RigUnit_GetCurveValue(Curve=R_blink_pose).Value) {variable=r_blink_val}
   n2 = RigVMVariableNode() {variable=l_blink_val}
   n3 = RigVMVariableNode() {variable=threshold}
   Less = /Script/RigVM.RigVMFunction_MathDoubleLess(B=0.75, A=@n2.Value) {is=struct RigVMUnitNode, template="Less::Execute(in A,in B,out Result)"}
   n4 = &RigVMFunction_ControlFlowBranch(
    Condition=&RigVMFunction_MathBoolAnd(
     B=&RigVMFunction_MathDoubleGreaterEqual(A=@n2.Value, B=@n3.Value) {template="GreaterEqual::Execute(in A,in B,out Result)"}.Result
     A=@Less.Result
    ).Result
    ExecuteContext=@RigVMFunction_Sequence.B
   )
   GreaterEqual = &RigVMFunction_MathDoubleGreaterEqual(B=0.75, A=RigVMVariableNode() {variable=l_blink_val}.Value) {template="GreaterEqual::Execute(in A,in B,out Result)"}
   RigVMFunction_ControlFlowBranch = &RigVMFunction_ControlFlowBranch(Condition=@GreaterEqual.Result, ExecuteContext=@n4.Completed)
   RigVMFunction_Sequence_1 = RigVMAggregateNode(ExecuteContext=@RigVMFunction_ControlFlowBranch.False)
   IsCurveActivatedFNC_1 = RigVMFunctionReferenceNode(
    Inputcurve=(Type=Curve,Name="L_wide_pose")
    ExecuteContext=@RigVMFunction_Sequence_1.A
    Threshold=RigVMVariableNode() {variable=threshold}.Value
   )
   RigVMFunction_MathBoolOr_1 = RigVMAggregateNode(
    B=RigVMVariableNode() {variable=l_brow_out_up}.Value
    A=RigVMVariableNode() {variable=l_brow_in_up}.Value
    C=@IsCurveActivatedFNC_1.Activated
   )
   n5 = &RigVMFunction_ControlFlowBranch(ExecuteContext=@IsCurveActivatedFNC_1.ExecuteContext, Condition=@RigVMFunction_MathBoolOr_1.Result)
   &RigUnit_SetCurveValue(Curve=flipbook_face_L_eye_index, Value=7, ExecutePin=@n5.True)
   n6 = &RigVMFunction_ControlFlowBranch(ExecuteContext=@RigVMFunction_Sequence_1.B, Condition=RigVMVariableNode() {variable=l_brow_out_dn}.Value)
   n7 = &RigVMFunction_ControlFlowBranch(ExecuteContext=@n6.Completed, Condition=RigVMVariableNode() {variable=l_brow_in_dn}.Value)
   RigVMFunction_MathBoolAnd = &RigVMFunction_MathBoolAnd(A=RigVMVariableNode() {variable=l_brow_in_dn}.Value, B=RigVMVariableNode() {variable=l_brow_out_dn}.Value)
   n8 = &RigVMFunction_ControlFlowBranch(ExecuteContext=@n7.Completed, Condition=@RigVMFunction_MathBoolAnd.Result)
   &RigUnit_SetCurveValue(Curve=flipbook_face_L_eye_index, Value=6, ExecutePin=@n8.True)
   &RigUnit_SetCurveValue(Curve=flipbook_face_L_eye_index, Value=5, ExecutePin=@n7.True)
   &RigUnit_SetCurveValue(Curve=flipbook_face_L_eye_index, Value=4, ExecutePin=@n6.True)
   n9 = &RigUnit_SetCurveValue(Curve=flipbook_face_L_eye_index, Value=3, ExecutePin=@RigVMFunction_ControlFlowBranch.True)
   RigVMFunction_MathBoolOr = &RigVMFunction_MathBoolOr(A=RigVMVariableNode() {variable=l_brow_in_dn}.Value, B=RigVMVariableNode() {variable=l_brow_out_dn}.Value)
   n10 = &RigVMFunction_ControlFlowBranch(Condition=@RigVMFunction_MathBoolOr.Result, ExecuteContext=@n9.ExecutePin)
   &RigUnit_SetCurveValue(Curve=flipbook_face_L_eye_index, Value=1, ExecutePin=@n10.True)
   &RigUnit_SetCurveValue(Curve=flipbook_face_L_eye_index, Value=2, ExecutePin=@n4.True)
   n11 = RigVMVariableNode() {variable=threshold}
   n12 = RigVMVariableNode() {variable=r_blink_val}
   n13 = &RigVMFunction_ControlFlowBranch(
    Condition=&RigVMFunction_MathBoolAnd(
     B=&RigVMFunction_MathDoubleGreaterEqual(B=@n11.Value, A=@n12.Value) {template="GreaterEqual::Execute(in A,in B,out Result)"}.Result
     A=&RigVMFunction_MathDoubleLess(B=0.75, A=@n12.Value) {template="Less::Execute(in A,in B,out Result)"}.Result
    ).Result
    ExecuteContext=@RigVMFunction_Sequence.C
   )
   n14 = &RigVMFunction_ControlFlowBranch(
    Condition=&RigVMFunction_MathDoubleGreaterEqual(B=0.75, A=RigVMVariableNode() {variable=r_blink_val}.Value) {template="GreaterEqual::Execute(in A,in B,out Result)"}.Result
    ExecuteContext=@n13.Completed
   )
   RigVMFunction_Sequence_2 = RigVMAggregateNode(ExecuteContext=@n14.False)
   IsCurveActivatedFNC_3 = RigVMFunctionReferenceNode(
    Inputcurve=(Type=Curve,Name="R_wide_pose")
    ExecuteContext=@RigVMFunction_Sequence_2.A
    Threshold=RigVMVariableNode() {variable=threshold}.Value
   )
   RigVMFunction_MathBoolOr_2 = RigVMAggregateNode(
    B=RigVMVariableNode() {variable=r_brow_out_up}.Value
    A=RigVMVariableNode() {variable=r_brow_in_up}.Value
    C=@IsCurveActivatedFNC_3.Activated
   )
   n15 = &RigVMFunction_ControlFlowBranch(ExecuteContext=@IsCurveActivatedFNC_3.ExecuteContext, Condition=@RigVMFunction_MathBoolOr_2.Result)
   &RigUnit_SetCurveValue(Curve=flipbook_face_R_eye_index, Value=7, ExecutePin=@n15.True)
   n16 = &RigVMFunction_ControlFlowBranch(ExecuteContext=@RigVMFunction_Sequence_2.B, Condition=RigVMVariableNode() {variable=r_brow_out_dn}.Value)
   n17 = &RigVMFunction_ControlFlowBranch(ExecuteContext=@n16.Completed, Condition=RigVMVariableNode() {variable=r_brow_in_dn}.Value)
   n18 = &RigVMFunction_ControlFlowBranch(
    ExecuteContext=@n17.Completed
    Condition=&RigVMFunction_MathBoolAnd(A=RigVMVariableNode() {variable=r_brow_in_dn}.Value, B=RigVMVariableNode() {variable=r_brow_out_dn}.Value).Result
   )
   &RigUnit_SetCurveValue(Curve=flipbook_face_R_eye_index, Value=6, ExecutePin=@n18.True)
   &RigUnit_SetCurveValue(Curve=flipbook_face_R_eye_index, Value=5, ExecutePin=@n17.True)
   &RigUnit_SetCurveValue(Curve=flipbook_face_R_eye_index, Value=4, ExecutePin=@n16.True)
   n19 = &RigUnit_SetCurveValue(Curve=flipbook_face_R_eye_index, Value=3, ExecutePin=@n14.True)
   n20 = &RigVMFunction_ControlFlowBranch(
    Condition=&RigVMFunction_MathBoolOr(A=RigVMVariableNode() {variable=r_brow_in_dn}.Value, B=RigVMVariableNode() {variable=r_brow_out_dn}.Value).Result
    ExecuteContext=@n19.ExecutePin
   )
   &RigUnit_SetCurveValue(Curve=flipbook_face_R_eye_index, Value=1, ExecutePin=@n20.True)
   &RigUnit_SetCurveValue(Curve=flipbook_face_R_eye_index, Value=2, ExecutePin=@n13.True)
  ]}
]
hierarchy [
 root {loc=<0 0 0>, rot=<0 0 0 1>}
 pelvis {parent=root, loc=<0 -5.9119 26.7164>, rot=<0.501 0.499 -0.501 -0.499>}
 spine_01 {parent=pelvis, loc=<0 0 0>, rot=<0 0 -0.0124 -0.9999>}
 spine_02 {parent=spine_01, loc=<10.023 -0.2496 0>, rot=<0 0 0.0482 -0.9988>}
 spine_03 {parent=spine_02, loc=<8.8513 0 0>, rot=<0 0 -0.0278 -0.9996>}
 neck_01 {parent=spine_03, loc=<1.5052 -2.1378 0>, rot=<0 0 0.5075 -0.8616>}
 neck_02 {parent=neck_01, loc=<7.9922 0 0>, rot=<0 0 0.0527 -0.9986>}
 head {parent=neck_02, loc=<5.4057 0 0>, rot=<-0.128 -0.6954 0.128 -0.6954>}
 face_attach {parent=head, loc=<0 0 0>, rot=<0 0 0 1>}
 reverse_jaw {parent=face_attach, loc=<0 1.2342 1.2103>, rot=<0 0.7071 0 -0.7071>}
 ear_01_l {parent=reverse_jaw, loc=<-1.575 -1.8011 -1.3569>, rot=<-0.0477 -0.203 0.2551 -0.9442>}
 ear_02_l {parent=ear_01_l, loc=<7.6075 0 0>, rot=<-0.6753 -0.0791 -0.1489 -0.718>}
 ear_03_l {parent=ear_02_l, loc=<4.653 0 0>, rot=<0.0043 -0.2469 -0.0259 -0.9687>}
 ear_04_l {parent=ear_03_l, loc=<8.6452 0 0>, rot=<-0.0484 -0.5541 0.0338 -0.8303>}
 ear_01_r {parent=reverse_jaw, loc=<-1.575 -1.8011 1.3569>, rot=<-0.2026 0.0548 0.9431 0.2579>}
 ear_02_r {parent=ear_01_r, loc=<-7.7127 0.009 0.381>, rot=<-0.7037 -0.0872 -0.1463 -0.6898>}
 ear_03_r {parent=ear_02_r, loc=<-9.3615 0.8912 2.5424>, rot=<-0.1738 -0.2474 0.0199 -0.953>}
 ear_04_r {parent=ear_03_r, loc=<0.1483 1.8269 3.8702>, rot=<0.0461 -0.8247 0.1687 -0.5379>}
 shoulder_fr_l {parent=spine_03, loc=<2.8386 4.7953 -6.3084>, rot=<-0.0036 -0.0032 0.8866 0.4625>}
 elbow_fr_l {parent=shoulder_fr_l, loc=<9.936 0 0>, rot=<-0.0003 -0.0016 0.331 -0.9436>}
 wrist_fr_l {parent=elbow_fr_l, loc=<9.6533 0 0>, rot=<0.996 -0.0824 0.0337 -0.0023>}
 ball_fr_l {parent=wrist_fr_l, loc=<2.586 0 0>, rot=<-0.0067 0.0516 -0.4874 -0.8716>}
 toe_fr_l {parent=ball_fr_l, loc=<2.4777 0 0>, rot=<-0.0266 0.0026 -0.106 -0.994>}
 toe_tip_fr_l {parent=toe_fr_l, loc=<2.3057 0 0>, rot=<0.0042 0.007 -0.033 -0.9994>}
 scapula_fr_l {parent=shoulder_fr_l, loc=<-6.9462 4.3433 -0.4324>, rot=<0.9608 -0.2758 0.0242 0.0112>}
 shoulder_fr_r {parent=spine_03, loc=<2.8386 4.7953 6.3084>, rot=<0.0032 -0.0036 0.4625 -0.8866>}
 elbow_fr_r {parent=shoulder_fr_r, loc=<-9.936 0 -0.0001>, rot=<-0.0003 -0.0016 0.331 -0.9436>}
 wrist_fr_r {parent=elbow_fr_r, loc=<-9.6533 0 0>, rot=<0.996 -0.0824 0.0337 -0.0023>}
 ball_fr_r {parent=wrist_fr_r, loc=<-2.586 0 0>, rot=<-0.0067 0.0516 -0.4874 -0.8716>}
 toe_fr_r {parent=ball_fr_r, loc=<-2.4778 0 0>, rot=<-0.0266 0.0026 -0.106 -0.994>}
 toe_tip_fr_r {parent=toe_fr_r, loc=<-2.3056 0 0.0001>, rot=<0.0042 0.007 -0.033 -0.9994>}
 scapula_fr_r {parent=shoulder_fr_r, loc=<6.9462 -4.3433 0.4324>, rot=<0.9608 -0.2758 0.0242 0.0112>}
 hips {parent=pelvis, loc=<0 0 0>, rot=<0 0 0 1>}
 thigh_bk_l {parent=hips, loc=<-7.2689 -2.206 -5.6646>, rot=<0.789 0.6145 0 0>}
 knee_bk_l {parent=thigh_bk_l, loc=<11.8525 0 0>, rot=<0 0 0.5452 -0.8383>}
 ankle_bk_l {parent=knee_bk_l, loc=<10.5441 0 0>, rot=<0 0 -0.4536 -0.8912>}
 ball_bk_l {parent=ankle_bk_l, loc=<8.945 0 0>, rot=<0 0 -0.5066 -0.8622>}
 toe_bk_l {parent=ball_bk_l, loc=<2.6569 0 0>, rot=<0 0 -0.1579 -0.9874>}
 toe_tip_bk_l {parent=toe_bk_l, loc=<3.137 0 0>, rot=<0 0 -0.0735 -0.9973>}
 thigh_bk_r {parent=hips, loc=<-7.2689 -2.206 5.6646>, rot=<0.6145 -0.789 0 0>}
 knee_bk_r {parent=thigh_bk_r, loc=<-11.8525 0.0001 0>, rot=<0 0 0.5452 -0.8383>}
 ankle_bk_r {parent=knee_bk_r, loc=<-10.5441 0 0>, rot=<0 0 -0.4536 -0.8912>}
 ball_bk_r {parent=ankle_bk_r, loc=<-8.9449 0 0>, rot=<0 0 -0.5066 -0.8622>}
 toe_bk_r {parent=ball_bk_r, loc=<-2.6569 0 0>, rot=<0 0 -0.1579 -0.9874>}
 toe_tip_bk_r {parent=toe_bk_r, loc=<-3.1371 0 0>, rot=<0 0 -0.0735 -0.9973>}
 tail_01 {parent=hips, loc=<-8.33 -5.1303 0>, rot=<0.257 -0.6588 0.6588 -0.257>}
 tail_02 {parent=tail_01, loc=<6.0512 0 1.219>, rot=<0 0.1989 0 -0.98>}
 C_nose_wrinkler_pose {type=Curve, loc=<0 0 0>, rot=<0 0 0 1>}
 L_blink_pose {type=Curve, loc=<0 0 0>, rot=<0 0 0 1>}
 L_brow_In_Dn_pose {type=Curve, loc=<0 0 0>, rot=<0 0 0 1>}
 L_brow_In_Up_pose {type=Curve, loc=<0 0 0>, rot=<0 0 0 1>}
 flipbook_face_L_eye_index {type=Curve, loc=<0 0 0>, rot=<0 0 0 1>}
 L_brow_Ot_Dn_pose {type=Curve, loc=<0 0 0>, rot=<0 0 0 1>}
 L_brow_Ot_Up_pose {type=Curve, loc=<0 0 0>, rot=<0 0 0 1>}
 L_brow_squeeze_in_pose {type=Curve, loc=<0 0 0>, rot=<0 0 0 1>}
 L_cheek_up_pose {type=Curve, loc=<0 0 0>, rot=<0 0 0 1>}
 L_lower_lid_up_pose {type=Curve, loc=<0 0 0>, rot=<0 0 0 1>}
 L_upperLidDown_pose {type=Curve, loc=<0 0 0>, rot=<0 0 0 1>}
 R_blink_pose {type=Curve, loc=<0 0 0>, rot=<0 0 0 1>}
 R_brow_In_Dn_pose {type=Curve, loc=<0 0 0>, rot=<0 0 0 1>}
 R_brow_In_Up_pose {type=Curve, loc=<0 0 0>, rot=<0 0 0 1>}
 R_brow_Ot_Dn_pose {type=Curve, loc=<0 0 0>, rot=<0 0 0 1>}
 R_brow_Ot_Up_pose {type=Curve, loc=<0 0 0>, rot=<0 0 0 1>}
 R_brow_squeeze_in_pose {type=Curve, loc=<0 0 0>, rot=<0 0 0 1>}
 R_cheek_up_pose {type=Curve, loc=<0 0 0>, rot=<0 0 0 1>}
 R_lower_lid_up_pose {type=Curve, loc=<0 0 0>, rot=<0 0 0 1>}
 R_upperLidDown_pose {type=Curve, loc=<0 0 0>, rot=<0 0 0 1>}
 R_wide_pose {type=Curve, loc=<0 0 0>, rot=<0 0 0 1>}
 flipbook_face_R_eye_index {type=Curve, loc=<0 0 0>, rot=<0 0 0 1>}
]
variables [
 l_blink_val {type=double, default=0}
 l_brow_in_dn {type=bool, default=False}
 l_brow_in_up {type=bool, default=False}
 l_brow_out_dn {type=bool, default=False}
 l_brow_out_up {type=bool, default=False}
 r_blink_val {type=double, default=0}
 r_brow_in_dn {type=bool, default=False}
 r_brow_in_up {type=bool, default=False}
 r_brow_out_dn {type=bool, default=False}
 r_brow_out_up {type=bool, default=False}
 threshold {type=double, default=0.35}
]
