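# Level Sequence that plays an animation flipbook for two quadruped cosmetic companion meshes, SitPlant and RipeQuiver, placed side by side.
# Each possessed skeletal mesh actor holds a fixed transform facing the same way and runs a skeletal animation track of eight back-to-back clips.
# Both tracks share identical section timing at 30 fps, so the two companions can be compared pose for pose, with one brief gap late in the timeline.
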LevelSequence /CosmeticCompanions/Assets/Quadruped/RipeQuiver/Meshes/Companion_Flipbook_LevelSeq
bindings [
 SitPlant_Quadruped {
  id=#1
  possessed_class=FortSkeletalMeshActor
  tracks [
   {
    type=MovieScene3DTransformTrack
    sections [{
      type=MovieScene3DTransformSection
      channels [
       Location.X {type=MovieSceneScriptingDoubleChannel, default=-6311.5234}
       Location.Y {type=MovieSceneScriptingDoubleChannel, default=-10313.6843}
       Location.Z {type=MovieSceneScriptingDoubleChannel, default=8}
       Rotation.X {type=MovieSceneScriptingDoubleChannel, default=0}
       Rotation.Y {type=MovieSceneScriptingDoubleChannel, default=0}
       Rotation.Z {type=MovieSceneScriptingDoubleChannel, default=180}
       Scale.X {type=MovieSceneScriptingDoubleChannel, default=1}
       Scale.Y {type=MovieSceneScriptingDoubleChannel, default=1}
       Scale.Z {type=MovieSceneScriptingDoubleChannel, default=1}
      ]
     }]
   }
   {
    type=MovieSceneSkeletalAnimationTrack
    sections [
     {type=MovieSceneSkeletalAnimationSection, end_frame=72, start_frame=0}
     {type=MovieSceneSkeletalAnimationSection, end_frame=414, start_frame=72}
     {type=MovieSceneSkeletalAnimationSection, end_frame=522, start_frame=414}
     {type=MovieSceneSkeletalAnimationSection, end_frame=680, start_frame=522}
     {type=MovieSceneSkeletalAnimationSection, end_frame=798, start_frame=680}
     {type=MovieSceneSkeletalAnimationSection, end_frame=880, start_frame=800}
     {type=MovieSceneSkeletalAnimationSection, end_frame=906, start_frame=880}
     {type=MovieSceneSkeletalAnimationSection, end_frame=962, start_frame=906}
    ]
   }
  ]
 }
 RipeQuiver_Quadruped {
  id=#2
  possessed_class=FortSkeletalMeshActor
  tracks [
   {
    type=MovieScene3DTransformTrack
    sections [{
      type=MovieScene3DTransformSection
      channels [
       Location.X {type=MovieSceneScriptingDoubleChannel, default=-6366.7412}
       Location.Y {type=MovieSceneScriptingDoubleChannel, default=-10312.8925}
       Location.Z {type=MovieSceneScriptingDoubleChannel, default=8}
       Rotation.X {type=MovieSceneScriptingDoubleChannel, default=0}
       Rotation.Y {type=MovieSceneScriptingDoubleChannel, default=0}
       Rotation.Z {type=MovieSceneScriptingDoubleChannel, default=180}
       Scale.X {type=MovieSceneScriptingDoubleChannel, default=1}
       Scale.Y {type=MovieSceneScriptingDoubleChannel, default=1}
       Scale.Z {type=MovieSceneScriptingDoubleChannel, default=1}
      ]
     }]
   }
   {
    type=MovieSceneSkeletalAnimationTrack
    sections [
     {type=MovieSceneSkeletalAnimationSection, end_frame=72, start_frame=0}
     {type=MovieSceneSkeletalAnimationSection, end_frame=414, start_frame=72}
     {type=MovieSceneSkeletalAnimationSection, end_frame=522, start_frame=414}
     {type=MovieSceneSkeletalAnimationSection, end_frame=680, start_frame=522}
     {type=MovieSceneSkeletalAnimationSection, end_frame=798, start_frame=680}
     {type=MovieSceneSkeletalAnimationSection, end_frame=880, start_frame=800}
     {type=MovieSceneSkeletalAnimationSection, end_frame=906, start_frame=880}
     {type=MovieSceneSkeletalAnimationSection, end_frame=962, start_frame=906}
    ]
   }
  ]
 }
]
display_rate {denominator=1, numerator=30}
tick_resolution {denominator=1, numerator=24000}
tracks []
