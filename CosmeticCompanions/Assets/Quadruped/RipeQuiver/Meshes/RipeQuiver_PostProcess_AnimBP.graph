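# Post-process Animation Blueprint for the RipeQuiver quadruped cosmetic companion mesh, layering an additive cosmetic pose over the incoming animation.
# A sequence player evaluates AdditivePoseAnim, and Apply Additive blends it onto the Input Pose, weighted by CosmeticPoseAlpha.
# The blueprint declares neither variable itself, and its event graph holds only unwired default nodes, so it contains no update logic.
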AnimBlueprint /CosmeticCompanions/Assets/Quadruped/RipeQuiver/Meshes/RipeQuiver_PostProcess_AnimBP
target_skeleton=&/RipeQuiver_Quadruped_Skeleton
graphs [
 AnimGraph {nodes [
   n1 = Get AdditivePoseAnim() -> (AdditivePoseAnim:Anim Sequence Object Reference)
   Output Pose(Result=Apply Additive(
     Base=Input Pose()
     Additive=Sequence Player!(Sequence:Anim Sequence Base Object Reference=@n1) -> (Pose)
     Alpha=Get CosmeticPoseAlpha() -> (CosmeticPoseAlpha:float)
    ) {alphaBoolBlend {blendOption=Linear}})
  ]}
 EventGraph {nodes [
   TryGetPawnOwner()
   Event BlueprintUpdateAnimation()
  ]}
]
variables []
